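# Niagara vehicle boost cosmetic that emits short-lived animated line meshes from the vehicle's boost sockets.
# Thin and curved emitter families, some GPU-simulated with CPU siblings, partition particles across filtered skeletal mesh sockets, scatter them in cylinders and stretch them by speed.
# Painted-variant colors override a default cyan or red-pink palette, while curve-driven dynamic material parameters animate each line and adjust painted glow.
# Spawn groups, game mode and preview flags vary spawning and placement.
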NiagaraSystem /VehicleCosmetics/Boosts/Boost_Animatedlines/NS_Boost_Animatedlines
dependencies {
 usedDataInterfaces [
  &NiagaraDataInterfaceArrayColor
  &NiagaraDataInterfaceMeshRendererInfo
  &NiagaraDataInterfaceArrayFloat3
  &NiagaraDataInterfaceCurve
  &NiagaraDataInterfaceColorCurve
  &NiagaraDataInterfaceSkeletalMesh
 ]
 usedDynamicInputs []
 usedModules [
  &self:SystemSpawnScript.SystemScriptSource.SystemScriptGraph.NiagaraNodeAssignment_0.SetVariables_#1
  &self:SystemSpawnScript.SystemScriptSource.SystemScriptGraph.NiagaraNodeAssignment_1.SetVariables_#2
  &self:SystemSpawnScript.SystemScriptSource.SystemScriptGraph.NiagaraNodeAssignment_4.SetVariables_#3
  &SystemState
  &self:SystemSpawnScript.SystemScriptSource.SystemScriptGraph.NiagaraNodeAssignment_2.SetVariables_#4
 ]
}
emitters {
 NE_VC_Boost_C3Thin {
  emitterUpdateScript [
   &EmitterState()
   &VC_FilteredSocketCount()
   &SpawnRate(SpawnRate=60, Use Spawn Probability=false)
   /SetVariables(Emitter.Density:NiagaraFloat=&MakeCustomFloatFromBool(Bool=$User.bIsPainted, True Float=0.1, False Float=2.3)) {isSetParameters=true, module=SetVariables_#5}
  ]
  particleSpawnScript [
   &VC_PaintedVariantsOverride(
    Black=<rgba 0 0 0 1>
    BurntSienna=&RandomRangeLinearColor(Minimum=<rgba 1 0.3394 0 1>, Maximum=<rgba 0.063 0.0091 0 1>)
    Cobalt=&RandomRangeLinearColor(Minimum=<rgba 0.001 0.2291 0.7969 1>, Maximum=<rgba 0 0.0028 0.151 1>)
    Crimson=&RandomRangeLinearColor(Minimum=<rgba 0.0885 0.0004 0 1>, Maximum=<rgba 0.6979 0.0028 0.0026 1>)
    DefaultColor=&SelectLinearColorFromArray(
     Array Sampling Mode=::NewEnumerator0
     Direct Array Index=?
     Color Selection Array=DataInterface() {propertyValues {ColorData [<rgba 0 1.6667 4 1>, <rgba 1.5 0 0.2 1>]}}
     Random Seed=0
    )
    ForestGreen=&RandomRangeLinearColor(Minimum=<rgba 0.001 0.0521 0.0006 1>, Maximum=<rgba 0 1.2 0.0606 1>)
    Grey=&RandomRangeLinearColor(Minimum=<rgba 0.007 0.007 0.0061 1>, Maximum=<rgba 0.07 0.07 0.07 1>)
    Lime=&RandomRangeLinearColor(Minimum=<rgba 0.0232 0.1043 0 1>, Maximum=<rgba 0.4089 0.8177 0 1>)
    Orange=&RandomRangeLinearColor(Minimum=<rgba 0.4397 0.0044 0 1>, Maximum=<rgba 1 0.1779 0.0103 1>)
    Pink=&RandomRangeLinearColor(Minimum=<rgba 0.1845 0.0037 0.0222 1>, Maximum=<rgba 0.9823 0.1779 0.9911 1>)
    Purple=&RandomRangeLinearColor(Minimum=<rgba 0.0595 0.0003 0.6654 1>, Maximum=<rgba 0.5391 0.0003 0.9911 1>)
    Saffron=&RandomRangeLinearColor(Minimum=<rgba 0.2135 0.2135 0.0014 1>, Maximum=<rgba 1 1 0.0048 1>)
    SkyBlue=&RandomRangeLinearColor(Minimum=<rgba 0.0021 0.0482 0.0742 1>, Maximum=<rgba 0.0497 1 0.9911 1>)
    TitaniumWhite=&RandomRangeLinearColor(Minimum=<rgba 0.3021 0.3021 0.3021 1>, Maximum=<rgba 0.9304 0.9304 0.9304 1>)
   )
   &InitializeParticle(
    Color Mode=::NewEnumerator0
    Lifetime Max=0.5
    Lifetime Min=0.3
    Lifetime Mode=::NewEnumerator1
    Mesh Scale=&Multiply_VectorByFloat(Vector=&RandomRangeVector(Minimum=<20 15 15>, Maximum=<35 20 20>), Float=1)
    Mesh Scale Mode=::NewEnumerator4
    Sprite Rotation Mode=::NewEnumerator2
    UsePositionOffset=false
    Write Material Random=true
   )
   &PartitionParticles(Number Of Partitions=$Emitter.PortCount)
   &SkeletalMeshLocation(
    Bone / Socket Index=$Particles.PartitionParticles.Partition
    Bone / Socket Sampling Mode=::NewEnumerator5
    MaskSpawnGroup=false
    Mesh Sampling Type=::NewEnumerator6
    Sampled Position Offset=&MakeCustomVectorFromBool(Bool=$User.bIsFrontendPreview, True Vector=<25 0 0>, False Vector=<65 0 0>)
    Skeletal Mesh=$System.SkeletalMesh
   )
   &AddVelocity(Velocity=<0 0 0>) {enabled=false}
   /SetVariables(Particles.VisibilityTag=&SpawnGroup()) {isSetParameters=true, module=SetVariables_#6}
   &InitialMeshOrientation(
    Mesh Axis To Orient=&RandomRangeVector(Minimum=<0 -1 1>, Maximum=<0 1 -1>)
    Mesh Orientation Mode=::NewEnumerator2
    Orientation Vector=&RandomRangeVector(Minimum=<0 1 1>, Maximum=<0 -1 1>, Randomness Mode=::NewEnumerator1)
    Rotation=<90 65 0>
    Rotation Coordinate Space=::NewEnumerator2
    Use Rotation Vector=false
   )
   &ShapeLocation(
    Cylinder Height=50
    Cylinder Height Midpoint=0.1
    Cylinder Radius=25
    MaskSpawnGroup=false
    Override Seed=false
    Rotation Mode=::NewEnumerator2
    Shape Primitive=::NewEnumerator1
    Surface Only=false
    UseOwnerScale=true
    Yaw / Pitch / Roll=<0 0.25 0>
   )
   /VehicleCosmetics/FX/VC_FXParents/Modules/DLR_ProductPreviews(GameModeID:NiagaraInt32=$System.GameModeID, bPreviewVehicleSpeed:NiagaraBool=$System.bPreviewVehicleSpeed)
  ]
  particleUpdateScript [
   &ParticleState()
   &ScaleMeshSize(Scale Factor=&SelectVectorFromArray(
     Direct Array Index=$Particles.PartitionParticles.Partition
     Vector Selection Array=DataInterface() {propertyValues {FloatData [<1 -0.75 0.75>, <1 -0.75 -0.75>]}}
    )) {enabled=false}
   &InheritVelocity(Speed Limit=true, UseSpeedThreshold=true) {enabled=false}
   &InheritSourceMovement(
    Applied Position Scale=&Multiply_VectorByFloat(
     Vector=<1 1 1>
     Float=&FloatFromCurve(FloatCurve=DataInterface() {propertyValues {Curve {
         defaultValue=1
         keys [
          {arriveTangent=-1.7885, interpMode=RCIM_Cubic, leaveTangent=-1.7885, tangentMode=RCTM_User, time=0.05, value=1}
          {interpMode=RCIM_Cubic, tangentMode=RCTM_User, time=1, value=0}
         ]
        }}})
    )
    ApplyForce=false
    ApplyPosition=true
    ApplyPositionFalloff=false
   ) {enabled=false, module=InheritSourceMovement001}
   &ScaleColor(
    Scale RGB=&VectorFromFloat(Value=&MakeCustomFloatFromBool(
      Bool=$User.bIsPainted
      True Float=&FloatFromCurve(
       FloatCurve=DataInterface() {propertyValues {Curve {keys [{interpMode=RCIM_Cubic, leaveTangent=-1.6688, tangentMode=RCTM_Break, time=0, value=1.25}, <at 1 0.15 RCIM_Cubic>]}}}
       Scale Curve=4
      )
      False Float=&FloatFromCurve(
       FloatCurve=DataInterface() {propertyValues {Curve {keys [
           {arriveTangent=-1.9978, interpMode=RCIM_Cubic, leaveTangent=-1.9978, tangentMode=RCTM_User, time=0, value=1}
           {
            arriveTangentWeight=0.311
            interpMode=RCIM_Cubic
            leaveTangent=-2.5749
            leaveTangentWeight=0.0777
            tangentMode=RCTM_Break
            time=0.8
            value=0.4
           }
           {arriveTangentWeight=0.2828, interpMode=RCIM_Cubic, leaveTangentWeight=0.0471, tangentMode=RCTM_User, time=1, value=0.2}
          ]}}}
       Scale Curve=&RandomRangeFloat(Minimum=3, Maximum=4)
      )
     ))
    ScaleA=true
    ScaleRGB=true
   ) {enabled=false}
   &DynamicMaterialParameters(
    Index 0 Param 1=&MaskFloatBySpawnGroup(
     Masked Float=&FloatFromCurve(
      FloatCurve=DataInterface() {propertyValues {Curve {keys [
          <at 0 0 RCIM_Cubic>
          {arriveTangent=-0.1296, interpMode=RCIM_Cubic, leaveTangent=-0.1296, tangentMode=RCTM_User, time=0.65, value=0.42}
          {arriveTangent=0.644, tangentMode=RCTM_User, time=1, value=1}
         ]}}}
      Scale Curve=0.8
     )
     Passthrough Float=&FloatFromCurve(
      FloatCurve=DataInterface() {propertyValues {Curve {keys [
          <at 0.1 0 RCIM_Cubic>
          {arriveTangent=-0.1296, interpMode=RCIM_Cubic, leaveTangent=-0.1296, tangentMode=RCTM_User, time=0.65, value=0.42}
          {arriveTangent=0.644, tangentMode=RCTM_User, time=1, value=1}
         ]}}}
      Scale Curve=0.8
     )
    )
    Index 0 Param 2=&MaskFloatBySpawnGroup(
     Masked Float=&FloatFromCurve(FloatCurve=DataInterface() {propertyValues {Curve {keys [
          <at 0 1 RCIM_Cubic>
          {arriveTangent=-0.1355, interpMode=RCIM_Cubic, leaveTangent=-0.1355, tangentMode=RCTM_User, time=0.7, value=0.07}
          <at 1 0 RCIM_Cubic>
         ]}}})
     Passthrough Float=&FloatFromCurve(FloatCurve=DataInterface() {propertyValues {Curve {keys [
          <at 0 1.1 RCIM_Cubic>
          {arriveTangent=-0.1355, interpMode=RCIM_Cubic, leaveTangent=-0.1355, tangentMode=RCTM_User, time=0.7, value=0.07}
          <at 1 0 RCIM_Cubic>
         ]}}})
    )
    Index 1 Param 1=&MakeCustomFloatFromBool(Bool=$User.bIsPainted, True Float=0, False Float=1)
    Index 1 Param 2=&MakeCustomFloatFromBool(Bool=$User.bIsPainted, True Float=&MakeFloatFromInt(Int=$User.Lines_Glow), False Float=1)
    Module.Index0Param0Write=true
    Module.Index0Param1Write=true
    Module.Index0Param2Write=true
    Module.Index0Param3Write=true
    Module.Index1Param0Write=true
    Module.Index1Param1Write=true
    Module.Index1Param2Write=true
    Module.Index1Param3Write=true
    Write Parameter Index 1=true
   ) {module=DynamicMaterialParameters001}
   &SolveForcesAndVelocity(Clamp Velocity=false, Limit Acceleration=false)
   &ScaleMeshSizeBySpeed(Max Scale Factor=<10 5 5>, Speed Threshold=100)
  ]
  rendererClasses [&NiagaraMeshRendererProperties, &NiagaraSpriteRendererProperties]
  renderers [
   {rendererClass=&NiagaraMeshRendererProperties, rendererIndex=0}
   {rendererClass=&NiagaraMeshRendererProperties, rendererIndex=1}
   {rendererClass=&NiagaraSpriteRendererProperties, rendererIndex=2}
  ]
 }
 NE_VC_Boost_C3Thin001 {
  emitterUpdateScript [
   &EmitterState()
   &VC_FilteredSocketCount()
   &SpawnRate(SpawnRate=40, Use Spawn Probability=false)
  ]
  particleSpawnScript [
   &VC_PaintedVariantsOverride(DefaultColor=&SelectLinearColorFromArray(
     Array Sampling Mode=::NewEnumerator0
     Direct Array Index=?
     Color Selection Array=DataInterface() {propertyValues {ColorData [<rgba 0 1.6667 4 1>, <rgba 1.5 0 0.2 1>]}}
     Random Seed=0
    ))
   &InitializeParticle(
    Color=$Particles.Color
    Lifetime Max=0.5
    Lifetime Min=0.3
    Lifetime Mode=::NewEnumerator1
    Mesh Scale=&Multiply_VectorByFloat(Vector=&RandomRangeVector(Minimum=<20 15 15>, Maximum=<35 20 20>), Float=1)
    Mesh Scale Mode=::NewEnumerator4
    Sprite Rotation Mode=::NewEnumerator2
    UsePositionOffset=false
    Write Material Random=true
   )
   &PartitionParticles(Number Of Partitions=$Emitter.PortCount)
   &SkeletalMeshLocation(
    Bone / Socket Index=$Particles.PartitionParticles.Partition
    Bone / Socket Sampling Mode=::NewEnumerator5
    MaskSpawnGroup=false
    Mesh Sampling Type=::NewEnumerator6
    Sampled Position Offset=&MakeCustomVectorFromBool(Bool=$User.bIsFrontendPreview, True Vector=<25 0 0>, False Vector=<65 0 0>)
    Skeletal Mesh=$System.SkeletalMesh
   )
   &AddVelocity(Velocity=<0 0 0>) {enabled=false}
   /SetVariables(Particles.VisibilityTag=&SpawnGroup()) {isSetParameters=true, module=SetVariables_#6}
   &InitialMeshOrientation(
    Mesh Axis To Orient=&RandomRangeVector(Minimum=<0 -1 1>, Maximum=<0 1 -1>)
    Mesh Orientation Mode=::NewEnumerator2
    Orientation Vector=&RandomRangeVector(Minimum=<0 1 1>, Maximum=<0 -1 1>, Randomness Mode=::NewEnumerator1)
    Rotation=<90 65 0>
    Rotation Coordinate Space=::NewEnumerator2
    Use Rotation Vector=false
   )
   &ShapeLocation(
    Cylinder Height=50
    Cylinder Height Midpoint=0.1
    Cylinder Radius=25
    MaskSpawnGroup=false
    Override Seed=false
    Rotation Mode=::NewEnumerator2
    Shape Primitive=::NewEnumerator1
    Surface Only=false
    UseOwnerScale=true
    Yaw / Pitch / Roll=<0 0.25 0>
   )
   &DLR_ProductPreviews(GameModeID:NiagaraInt32=$System.GameModeID, bPreviewVehicleSpeed:NiagaraBool=$System.bPreviewVehicleSpeed)
  ]
  particleUpdateScript [
   &ParticleState()
   &ScaleMeshSize(Scale Factor=&SelectVectorFromArray(
     Direct Array Index=$Particles.PartitionParticles.Partition
     Vector Selection Array=DataInterface() {propertyValues {FloatData [<1 -0.75 0.75>, <1 -0.75 -0.75>]}}
    )) {enabled=false}
   &InheritVelocity(Speed Limit=true, UseSpeedThreshold=true) {enabled=false}
   &InheritSourceMovement(
    Applied Position Scale=&Multiply_VectorByFloat(
     Vector=<1 1 1>
     Float=&FloatFromCurve(FloatCurve=DataInterface() {propertyValues {Curve {
         defaultValue=1
         keys [
          {arriveTangent=-1.7885, interpMode=RCIM_Cubic, leaveTangent=-1.7885, tangentMode=RCTM_User, time=0.05, value=1}
          {interpMode=RCIM_Cubic, tangentMode=RCTM_User, time=1, value=0}
         ]
        }}})
    )
    ApplyForce=false
    ApplyPosition=true
    ApplyPositionFalloff=false
   ) {enabled=false, module=InheritSourceMovement001}
   &ScaleColor(
    Scale RGB=&VectorFromFloat(Value=&MakeCustomFloatFromBool(
      Bool=$User.bIsPainted
      True Float=&FloatFromCurve(
       FloatCurve=DataInterface() {propertyValues {Curve {keys [{interpMode=RCIM_Cubic, leaveTangent=-1.6688, tangentMode=RCTM_Break, time=0, value=1.25}, <at 1 0.15 RCIM_Cubic>]}}}
       Scale Curve=4
      )
      False Float=&FloatFromCurve(
       FloatCurve=DataInterface() {propertyValues {Curve {keys [
           {arriveTangent=-1.9978, interpMode=RCIM_Cubic, leaveTangent=-1.9978, tangentMode=RCTM_User, time=0, value=1}
           {
            arriveTangentWeight=0.311
            interpMode=RCIM_Cubic
            leaveTangent=-2.5749
            leaveTangentWeight=0.0777
            tangentMode=RCTM_Break
            time=0.8
            value=0.4
           }
           {arriveTangentWeight=0.2828, interpMode=RCIM_Cubic, leaveTangentWeight=0.0471, tangentMode=RCTM_User, time=1, value=0.2}
          ]}}}
       Scale Curve=&RandomRangeFloat(Minimum=3, Maximum=4)
      )
     ))
    ScaleA=true
    ScaleRGB=true
   ) {enabled=false}
   &DynamicMaterialParameters(
    Index 0 Param 1=&FloatFromCurve(
     FloatCurve=DataInterface() {propertyValues {Curve {keys [
         <at 0.1 0 RCIM_Cubic>
         {arriveTangent=-0.1296, interpMode=RCIM_Cubic, leaveTangent=-0.1296, tangentMode=RCTM_User, time=0.65, value=0.42}
         {arriveTangent=0.644, tangentMode=RCTM_User, time=1, value=1}
        ]}}}
     Scale Curve=0.8
    )
    Index 0 Param 2=&FloatFromCurve(FloatCurve=DataInterface() {propertyValues {Curve {keys [
         <at 0 1.1 RCIM_Cubic>
         {arriveTangent=-0.1355, interpMode=RCIM_Cubic, leaveTangent=-0.1355, tangentMode=RCTM_User, time=0.7, value=0.07}
         <at 1 0 RCIM_Cubic>
        ]}}})
    Index 1 Param 1=&MakeCustomFloatFromBool(Bool=$User.bIsPainted, True Float=0, False Float=1)
    Index 1 Param 2=&MakeCustomFloatFromBool(Bool=$User.bIsPainted, True Float=&MakeFloatFromInt(Int=$User.Lines_Glow), False Float=1)
    Module.Index0Param0Write=true
    Module.Index0Param1Write=true
    Module.Index0Param2Write=true
    Module.Index0Param3Write=true
    Module.Index1Param0Write=true
    Module.Index1Param1Write=true
    Module.Index1Param2Write=true
    Module.Index1Param3Write=true
    Write Parameter Index 1=true
   ) {module=DynamicMaterialParameters001}
   &SolveForcesAndVelocity(Clamp Velocity=false, Limit Acceleration=false)
   &ScaleMeshSizeBySpeed(Max Scale Factor=<10 5 5>, Speed Threshold=100)
  ]
  rendererClasses [&NiagaraMeshRendererProperties, &NiagaraSpriteRendererProperties]
  renderers [
   {rendererClass=&NiagaraMeshRendererProperties, rendererIndex=0}
   {rendererClass=&NiagaraMeshRendererProperties, rendererIndex=1}
   {rendererClass=&NiagaraSpriteRendererProperties, rendererIndex=2}
  ]
 }
 NE_VC_Boost_Curve2 {
  simTarget=GPUComputeSim
  emitterUpdateScript [
   &EmitterState()
   &SpawnRate(SpawnGroup=1, SpawnRate=20, Use Spawn Probability=false) {module=SpawnRate002}
   &VC_FilteredSocketCount()
   &SpawnRate(
    SpawnGroup=1
    SpawnRate=&Multiply_Float(
     A=$Engine.Emitter.SpawnCountScale
     B=/VehicleCosmetics/FX/Niagara/Modules/VC_Mask_Floats(Coro Value:NiagaraFloat=75, FrontEnd Value:NiagaraFloat=1, FN Value:NiagaraFloat=0.5)
    )
    Use Spawn Probability=false
   ) {module=SpawnRate003}
   &SpawnRate(SpawnRate=30, Use Spawn Probability=false)
   /SetVariables(Emitter.Density:NiagaraFloat=&MakeCustomFloatFromBool(Bool=$User.bIsPainted, True Float=0.1, False Float=2.3)) {isSetParameters=true, module=SetVariables_#5}
  ]
  particleSpawnScript [
   &VC_PaintedVariantsOverride(
    Black=<rgba 0 0 0 1>
    BurntSienna=&RandomRangeLinearColor(Minimum=<rgba 1 0.3394 0 1>, Maximum=<rgba 0.063 0.0091 0 1>)
    Cobalt=&RandomRangeLinearColor(Minimum=<rgba 0.001 0.2291 0.7969 1>, Maximum=<rgba 0 0.0028 0.151 1>)
    Crimson=&RandomRangeLinearColor(Minimum=<rgba 0.0885 0.0004 0 1>, Maximum=<rgba 0.6979 0.0028 0.0026 1>)
    DefaultColor=&SelectLinearColorFromArray(
     Array Sampling Mode=::NewEnumerator0
     Direct Array Index=?
     Color Selection Array=DataInterface() {propertyValues {ColorData [<rgba 0 1.6667 4 1>, <rgba 1.5 0 0.2 1>]}}
     Random Seed=0
    )
    ForestGreen=&RandomRangeLinearColor(Minimum=<rgba 0.001 0.0521 0.0006 1>, Maximum=<rgba 0 1.2 0.0606 1>)
    Grey=&RandomRangeLinearColor(Minimum=<rgba 0.007 0.007 0.0061 1>, Maximum=<rgba 0.07 0.07 0.07 1>)
    Lime=&RandomRangeLinearColor(Minimum=<rgba 0.0232 0.1043 0 1>, Maximum=<rgba 0.4089 0.8177 0 1>)
    Orange=&RandomRangeLinearColor(Minimum=<rgba 0.4397 0.0044 0 1>, Maximum=<rgba 1 0.1779 0.0103 1>)
    Pink=&RandomRangeLinearColor(Minimum=<rgba 0.1845 0.0037 0.0222 1>, Maximum=<rgba 0.9823 0.1779 0.9911 1>)
    Purple=&RandomRangeLinearColor(Minimum=<rgba 0.0595 0.0003 0.6654 1>, Maximum=<rgba 0.5391 0.0003 0.9911 1>)
    Saffron=&RandomRangeLinearColor(Minimum=<rgba 0.2135 0.2135 0.0014 1>, Maximum=<rgba 1 1 0.0048 1>)
    SkyBlue=&RandomRangeLinearColor(Minimum=<rgba 0.0021 0.0482 0.0742 1>, Maximum=<rgba 0.0497 1 0.9911 1>)
    TitaniumWhite=&RandomRangeLinearColor(Minimum=<rgba 0.3021 0.3021 0.3021 1>, Maximum=<rgba 0.9304 0.9304 0.9304 1>)
   )
   /SetVariables(Particles.VisibilityTag=&SpawnGroup()) {isSetParameters=true, module=SetVariables_#7}
   &InitializeParticle(
    Color Mode=::NewEnumerator0
    Lifetime=&MaskFloatBySpawnGroup(
     Masked Float=&RandomRangeFloat(Minimum=0.2, Maximum=0.5)
     Passthrough Float=&RandomRangeFloat(Minimum=0.2, Maximum=0.3)
    )
    Mesh Scale=&Multiply_Vector(
     A=&MaskVectorBySpawnGroup(
      Masked Vector=&RandomRangeVector(Minimum=<7 8 8>, Maximum=<10 11 11>)
      Passthrough Vector=&RandomRangeVector(Minimum=<6 11 11>, Maximum=<12 15 15>)
     )
     B=&MakeVector(X=2.5, Y=1, Z=1)
    )
    Mesh Scale Mode=::NewEnumerator4
    Sprite Rotation Mode=::NewEnumerator2
    UsePositionOffset=false
    Write Material Random=true
   )
   &PartitionParticles(Number Of Partitions=$Emitter.PortCount)
   &SkeletalMeshLocation(
    Bone / Socket Index=$Particles.PartitionParticles.Partition
    Bone / Socket Sampling Mode=::NewEnumerator5
    MaskSpawnGroup=false
    Mesh Sampling Type=::NewEnumerator6
    Sampled Position Offset=&MakeCustomVectorFromBool(Bool=$User.bIsFrontendPreview, True Vector=<30 0 0>, False Vector=<65 0 0>)
    Skeletal Mesh=$System.SkeletalMesh
   )
   &AddVelocity(Velocity=<0 0 0>) {enabled=false}
   &InitialMeshOrientation(
    Mesh Axis To Orient=&RandomRangeVector(Minimum=<0 -1 1>, Maximum=<0 1 -1>)
    Mesh Orientation Mode=::NewEnumerator2
    Orientation Vector=&RandomRangeVector(Minimum=<0 1 1>, Maximum=<0 -1 1>, Randomness Mode=::NewEnumerator1)
    Rotation=<90 65 0>
    Rotation Coordinate Space=::NewEnumerator2
    Use Rotation Vector=false
   )
   &ShapeLocation(
    Cylinder Height=50
    Cylinder Height Midpoint=0.1
    Cylinder Radius=25
    MaskSpawnGroup=false
    Override Seed=false
    Rotation Mode=::NewEnumerator2
    Shape Primitive=::NewEnumerator1
    Surface Only=false
    UseOwnerScale=true
    Yaw / Pitch / Roll=<0 0.25 0>
   )
   &DLR_ProductPreviews(GameModeID:NiagaraInt32=$System.GameModeID, bPreviewVehicleSpeed:NiagaraBool=$System.bPreviewVehicleSpeed)
  ]
  particleUpdateScript [
   &ParticleState()
   &ScaleMeshSize(Scale Factor=&SelectVectorFromArray(
     Direct Array Index=$Particles.PartitionParticles.Partition
     Vector Selection Array=DataInterface() {propertyValues {FloatData [<1 -0.75 0.75>, <1 -0.75 -0.75>]}}
    )) {enabled=false}
   &InheritSourceMovement(
    Applied Position Scale=&Multiply_VectorByFloat(
     Vector=<1 1 1>
     Float=&FloatFromCurve(FloatCurve=DataInterface() {propertyValues {Curve {
         defaultValue=1
         keys [
          {arriveTangent=-1.7885, interpMode=RCIM_Cubic, leaveTangent=-1.7885, tangentMode=RCTM_User, time=0.05, value=1}
          {interpMode=RCIM_Cubic, tangentMode=RCTM_User, time=1, value=0}
         ]
        }}})
    )
    ApplyForce=false
    ApplyPosition=true
    ApplyPositionFalloff=false
   ) {enabled=false, module=InheritSourceMovement001}
   &ScaleColor(
    Scale RGB=&VectorFromFloat(Value=&MakeCustomFloatFromBool(
      Bool=$User.bIsPainted
      True Float=&FloatFromCurve(
       FloatCurve=DataInterface() {propertyValues {Curve {keys [{interpMode=RCIM_Cubic, leaveTangent=-1.6688, tangentMode=RCTM_Break, time=0, value=1.25}, <at 1 0.15 RCIM_Cubic>]}}}
       Scale Curve=4
      )
      False Float=&FloatFromCurve(
       FloatCurve=DataInterface() {propertyValues {Curve {keys [
           {arriveTangent=-1.9978, interpMode=RCIM_Cubic, leaveTangent=-1.9978, tangentMode=RCTM_User, time=0, value=1}
           {
            arriveTangentWeight=0.311
            interpMode=RCIM_Cubic
            leaveTangent=-2.5749
            leaveTangentWeight=0.0777
            tangentMode=RCTM_Break
            time=0.8
            value=0.4
           }
           {arriveTangentWeight=0.2828, interpMode=RCIM_Cubic, leaveTangentWeight=0.0471, tangentMode=RCTM_User, time=1, value=0.2}
          ]}}}
       Scale Curve=&RandomRangeFloat(Minimum=3, Maximum=4)
      )
     ))
    ScaleA=true
    ScaleRGB=true
   ) {enabled=false}
   &DynamicMaterialParameters(
    Index 0 Param 1=&FloatFromCurve(
     FloatCurve=DataInterface() {propertyValues {Curve {keys [
         <at 0.1 0>
         {arriveTangent=1.8777, leaveTangent=1.8777, time=0.67, value=0.38}
         {arriveTangent=0.644, tangentMode=RCTM_User, time=1, value=1}
        ]}}}
     Scale Curve=0.8
    )
    Index 0 Param 2=&FloatFromCurve(FloatCurve=DataInterface() {propertyValues {Curve {keys [
         <at 0 1.1 RCIM_Cubic>
         {arriveTangent=-0.1355, interpMode=RCIM_Cubic, leaveTangent=-0.1355, tangentMode=RCTM_User, time=0.52, value=0.19}
         <at 1 0 RCIM_Cubic>
        ]}}})
    Index 1 Param 1=&MakeCustomFloatFromBool(Bool=$User.bIsPainted, True Float=0, False Float=1)
    Index 1 Param 2=&MakeCustomFloatFromBool(Bool=$User.bIsPainted, True Float=&MakeFloatFromInt(Int=$User.Lines_Glow), False Float=1)
    Module.Index0Param0Write=true
    Module.Index0Param1Write=true
    Module.Index0Param2Write=true
    Module.Index0Param3Write=true
    Module.Index1Param0Write=true
    Module.Index1Param1Write=true
    Module.Index1Param2Write=true
    Module.Index1Param3Write=true
    Write Parameter Index 1=true
   ) {module=DynamicMaterialParameters001}
   &SolveForcesAndVelocity(Clamp Velocity=false, Limit Acceleration=false)
   &ScaleMeshSizeBySpeed(Max Scale Factor=<10 5 5>, Speed Threshold=100)
  ]
  rendererClasses [&NiagaraMeshRendererProperties, &NiagaraSpriteRendererProperties]
  renderers [
   {rendererClass=&NiagaraMeshRendererProperties, rendererIndex=0}
   {rendererClass=&NiagaraMeshRendererProperties, rendererIndex=1}
   {rendererClass=&NiagaraMeshRendererProperties, rendererIndex=2}
   {rendererClass=&NiagaraMeshRendererProperties, rendererIndex=3}
   {rendererClass=&NiagaraSpriteRendererProperties, rendererIndex=4}
  ]
 }
 NE_VC_Boost_Curve2001 {
  emitterUpdateScript [
   &EmitterState()
   &SpawnRate(SpawnGroup=1, SpawnRate=20, Use Spawn Probability=false) {module=SpawnRate002}
   &VC_FilteredSocketCount()
   &SpawnRate(
    SpawnGroup=1
    SpawnRate=&Multiply_Float(
     A=$Engine.Emitter.SpawnCountScale
     B=&VC_Mask_Floats(Coro Value:NiagaraFloat=25, FrontEnd Value:NiagaraFloat=1, FN Value:NiagaraFloat=0.5)
    )
    Use Spawn Probability=false
   ) {module=SpawnRate003}
   &SpawnRate(SpawnRate=30, Use Spawn Probability=false)
  ]
  particleSpawnScript [
   &VC_PaintedVariantsOverride(DefaultColor=&SelectLinearColorFromArray(
     Array Sampling Mode=::NewEnumerator0
     Direct Array Index=?
     Color Selection Array=DataInterface() {propertyValues {ColorData [<rgba 0 1.6667 4 1>, <rgba 1.5 0 0.2 1>]}}
     Random Seed=0
    ))
   /SetVariables(Particles.VisibilityTag=&SpawnGroup()) {isSetParameters=true, module=SetVariables_#7}
   &InitializeParticle(
    Color=$Particles.Color
    Lifetime=&MaskFloatBySpawnGroup(
     Masked Float=&RandomRangeFloat(Minimum=0.2, Maximum=0.5)
     Passthrough Float=&RandomRangeFloat(Minimum=0.2, Maximum=0.3)
    )
    Mesh Scale=&Multiply_Vector(
     A=&MaskVectorBySpawnGroup(
      Masked Vector=&RandomRangeVector(Minimum=<7 8 8>, Maximum=<10 11 11>)
      Passthrough Vector=&RandomRangeVector(Minimum=<6 11 11>, Maximum=<12 15 15>)
     )
     B=&MakeVector(X=2.5, Y=1, Z=1)
    )
    Mesh Scale Mode=::NewEnumerator4
    Sprite Rotation Mode=::NewEnumerator2
    UsePositionOffset=false
    Write Material Random=true
   )
   &PartitionParticles(Number Of Partitions=$Emitter.PortCount)
   &SkeletalMeshLocation(
    Bone / Socket Index=$Particles.PartitionParticles.Partition
    Bone / Socket Sampling Mode=::NewEnumerator5
    MaskSpawnGroup=false
    Mesh Sampling Type=::NewEnumerator6
    Sampled Position Offset=&MakeCustomVectorFromBool(Bool=$User.bIsFrontendPreview, True Vector=<30 0 0>, False Vector=<65 0 0>)
    Skeletal Mesh=$System.SkeletalMesh
   )
   &AddVelocity(Velocity=<0 0 0>) {enabled=false}
   &InitialMeshOrientation(
    Mesh Axis To Orient=&RandomRangeVector(Minimum=<0 -1 1>, Maximum=<0 1 -1>)
    Mesh Orientation Mode=::NewEnumerator2
    Orientation Vector=&RandomRangeVector(Minimum=<0 1 1>, Maximum=<0 -1 1>, Randomness Mode=::NewEnumerator1)
    Rotation=<90 65 0>
    Rotation Coordinate Space=::NewEnumerator2
    Use Rotation Vector=false
   )
   &ShapeLocation(
    Cylinder Height=50
    Cylinder Height Midpoint=0.1
    Cylinder Radius=25
    MaskSpawnGroup=false
    Override Seed=false
    Rotation Mode=::NewEnumerator2
    Shape Primitive=::NewEnumerator1
    Surface Only=false
    UseOwnerScale=true
    Yaw / Pitch / Roll=<0 0.25 0>
   )
   &DLR_ProductPreviews(GameModeID:NiagaraInt32=$System.GameModeID, bPreviewVehicleSpeed:NiagaraBool=$System.bPreviewVehicleSpeed)
  ]
  particleUpdateScript [
   &ParticleState()
   &ScaleMeshSize(Scale Factor=&SelectVectorFromArray(
     Direct Array Index=$Particles.PartitionParticles.Partition
     Vector Selection Array=DataInterface() {propertyValues {FloatData [<1 -0.75 0.75>, <1 -0.75 -0.75>]}}
    )) {enabled=false}
   &InheritSourceMovement(
    Applied Position Scale=&Multiply_VectorByFloat(
     Vector=<1 1 1>
     Float=&FloatFromCurve(FloatCurve=DataInterface() {propertyValues {Curve {
         defaultValue=1
         keys [
          {arriveTangent=-1.7885, interpMode=RCIM_Cubic, leaveTangent=-1.7885, tangentMode=RCTM_User, time=0.05, value=1}
          {interpMode=RCIM_Cubic, tangentMode=RCTM_User, time=1, value=0}
         ]
        }}})
    )
    ApplyForce=false
    ApplyPosition=true
    ApplyPositionFalloff=false
   ) {enabled=false, module=InheritSourceMovement001}
   &ScaleColor(
    Scale RGB=&VectorFromFloat(Value=&MakeCustomFloatFromBool(
      Bool=$User.bIsPainted
      True Float=&FloatFromCurve(
       FloatCurve=DataInterface() {propertyValues {Curve {keys [{interpMode=RCIM_Cubic, leaveTangent=-1.6688, tangentMode=RCTM_Break, time=0, value=1.25}, <at 1 0.15 RCIM_Cubic>]}}}
       Scale Curve=4
      )
      False Float=&FloatFromCurve(
       FloatCurve=DataInterface() {propertyValues {Curve {keys [
           {arriveTangent=-1.9978, interpMode=RCIM_Cubic, leaveTangent=-1.9978, tangentMode=RCTM_User, time=0, value=1}
           {
            arriveTangentWeight=0.311
            interpMode=RCIM_Cubic
            leaveTangent=-2.5749
            leaveTangentWeight=0.0777
            tangentMode=RCTM_Break
            time=0.8
            value=0.4
           }
           {arriveTangentWeight=0.2828, interpMode=RCIM_Cubic, leaveTangentWeight=0.0471, tangentMode=RCTM_User, time=1, value=0.2}
          ]}}}
       Scale Curve=&RandomRangeFloat(Minimum=3, Maximum=4)
      )
     ))
    ScaleA=true
    ScaleRGB=true
   ) {enabled=false}
   &DynamicMaterialParameters(
    Index 0 Param 1=&FloatFromCurve(
     FloatCurve=DataInterface() {propertyValues {Curve {keys [
         <at 0.1 0>
         {arriveTangent=1.8777, leaveTangent=1.8777, time=0.67, value=0.38}
         {arriveTangent=0.644, tangentMode=RCTM_User, time=1, value=1}
        ]}}}
     Scale Curve=0.8
    )
    Index 0 Param 2=&FloatFromCurve(FloatCurve=DataInterface() {propertyValues {Curve {keys [
         <at 0 1.1 RCIM_Cubic>
         {arriveTangent=-0.1355, interpMode=RCIM_Cubic, leaveTangent=-0.1355, tangentMode=RCTM_User, time=0.52, value=0.19}
         <at 1 0 RCIM_Cubic>
        ]}}})
    Index 1 Param 1=&MakeCustomFloatFromBool(Bool=$User.bIsPainted, True Float=0, False Float=1)
    Index 1 Param 2=&MakeCustomFloatFromBool(Bool=$User.bIsPainted, True Float=&MakeFloatFromInt(Int=$User.Lines_Glow), False Float=1)
    Module.Index0Param0Write=true
    Module.Index0Param1Write=true
    Module.Index0Param2Write=true
    Module.Index0Param3Write=true
    Module.Index1Param0Write=true
    Module.Index1Param1Write=true
    Module.Index1Param2Write=true
    Module.Index1Param3Write=true
    Write Parameter Index 1=true
   ) {module=DynamicMaterialParameters001}
   &SolveForcesAndVelocity(Clamp Velocity=false, Limit Acceleration=false)
   &ScaleMeshSizeBySpeed(Max Scale Factor=<10 5 5>, Speed Threshold=100)
  ]
  rendererClasses [&NiagaraMeshRendererProperties, &NiagaraSpriteRendererProperties]
  renderers [
   {rendererClass=&NiagaraMeshRendererProperties, rendererIndex=0}
   {rendererClass=&NiagaraMeshRendererProperties, rendererIndex=1}
   {rendererClass=&NiagaraMeshRendererProperties, rendererIndex=2}
   {rendererClass=&NiagaraMeshRendererProperties, rendererIndex=3}
   {rendererClass=&NiagaraSpriteRendererProperties, rendererIndex=4}
  ]
 }
 NE_VC_Boost_Curve3 {
  simTarget=GPUComputeSim
  emitterUpdateScript [
   &EmitterState()
   &VC_FilteredSocketCount()
   &SpawnRate(
    SpawnGroup=1
    SpawnRate=&Multiply_Float(
     A=$Engine.Emitter.SpawnCountScale
     B=&VC_Mask_Floats(Coro Value:NiagaraFloat=75, FrontEnd Value:NiagaraFloat=1, FN Value:NiagaraFloat=0.5)
    )
    Use Spawn Probability=false
   ) {module=SpawnRate003}
   &SpawnRate(SpawnRate=40, Use Spawn Probability=false)
   &SpawnRate(SpawnGroup=1, SpawnRate=100, Use Spawn Probability=false) {module=SpawnRate002}
   /SetVariables(Emitter.Density:NiagaraFloat=&MakeCustomFloatFromBool(Bool=$User.bIsPainted, True Float=0.1, False Float=2.3)) {isSetParameters=true, module=SetVariables_#5}
  ]
  particleSpawnScript [
   &VC_PaintedVariantsOverride(
    Black=<rgba 0 0 0 1>
    BurntSienna=&RandomRangeLinearColor(Minimum=<rgba 1 0.3394 0 1>, Maximum=<rgba 0.063 0.0091 0 1>)
    Cobalt=&RandomRangeLinearColor(Minimum=<rgba 0.001 0.2291 0.7969 1>, Maximum=<rgba 0 0.0028 0.151 1>)
    Crimson=&RandomRangeLinearColor(Minimum=<rgba 0.0885 0.0004 0 1>, Maximum=<rgba 0.6979 0.0028 0.0026 1>)
    DefaultColor=&SelectLinearColorFromArray(
     Array Sampling Mode=::NewEnumerator0
     Direct Array Index=?
     Color Selection Array=DataInterface() {propertyValues {ColorData [<rgba 0 1.6667 4 1>, <rgba 1.5 0 0.2 1>]}}
     Random Seed=0
    )
    ForestGreen=&RandomRangeLinearColor(Minimum=<rgba 0.001 0.0521 0.0006 1>, Maximum=<rgba 0 1.2 0.0606 1>)
    Grey=&RandomRangeLinearColor(Minimum=<rgba 0.007 0.007 0.0061 1>, Maximum=<rgba 0.07 0.07 0.07 1>)
    Lime=&RandomRangeLinearColor(Minimum=<rgba 0.0232 0.1043 0 1>, Maximum=<rgba 0.4089 0.8177 0 1>)
    Orange=&RandomRangeLinearColor(Minimum=<rgba 0.4397 0.0044 0 1>, Maximum=<rgba 1 0.1779 0.0103 1>)
    Pink=&RandomRangeLinearColor(Minimum=<rgba 0.1845 0.0037 0.0222 1>, Maximum=<rgba 0.9823 0.1779 0.9911 1>)
    Purple=&RandomRangeLinearColor(Minimum=<rgba 0.0595 0.0003 0.6654 1>, Maximum=<rgba 0.5391 0.0003 0.9911 1>)
    Saffron=&RandomRangeLinearColor(Minimum=<rgba 0.2135 0.2135 0.0014 1>, Maximum=<rgba 1 1 0.0048 1>)
    SkyBlue=&RandomRangeLinearColor(Minimum=<rgba 0.0021 0.0482 0.0742 1>, Maximum=<rgba 0.0497 1 0.9911 1>)
    TitaniumWhite=&RandomRangeLinearColor(Minimum=<rgba 0.3021 0.3021 0.3021 1>, Maximum=<rgba 0.9304 0.9304 0.9304 1>)
   )
   &InitializeParticle(
    Color Mode=::NewEnumerator0
    Lifetime=&MaskFloatBySpawnGroup(
     Masked Float=&RandomRangeFloat(Minimum=0.1, Maximum=0.4)
     Passthrough Float=&RandomRangeFloat(Minimum=0.1, Maximum=0.3)
    )
    Mesh Scale=&Multiply_Vector(
     A=&MaskVectorBySpawnGroup(
      Masked Vector=&RandomRangeVector(Minimum=<7 6 6>, Maximum=<15 10 10>)
      Passthrough Vector=&RandomRangeVector(Minimum=<7 7 7>, Maximum=<9 9 9>)
     )
     B=&MakeVector(X=&RandomRangeFloat(Minimum=2, Maximum=3), Y=1, Z=1)
    )
    Mesh Scale Mode=::NewEnumerator4
    Sprite Rotation Mode=::NewEnumerator2
    UsePositionOffset=false
    Write Material Random=true
   )
   &PartitionParticles(Number Of Partitions=$Emitter.PortCount)
   &SkeletalMeshLocation(
    Bone / Socket Index=$Particles.PartitionParticles.Partition
    Bone / Socket Sampling Mode=::NewEnumerator5
    MaskSpawnGroup=false
    Mesh Sampling Type=::NewEnumerator6
    Sampled Position Offset=&MakeCustomVectorFromBool(Bool=$User.bIsFrontendPreview, True Vector=<32 0 0>, False Vector=<65 0 0>)
    Skeletal Mesh=$System.SkeletalMesh
   )
   &AddVelocity(Velocity=<0 0 0>) {enabled=false}
   /SetVariables(Particles.VisibilityTag=&SpawnGroup()) {isSetParameters=true, module=SetVariables_#6}
   &InitialMeshOrientation(
    Mesh Axis To Orient=&RandomRangeVector(Minimum=<0 -1 1>, Maximum=<0 1 -1>)
    Mesh Orientation Mode=::NewEnumerator2
    Orientation Vector=&RandomRangeVector(Minimum=<0 1 1>, Maximum=<0 -1 1>, Randomness Mode=::NewEnumerator1)
    Rotation=<90 65 0>
    Rotation Coordinate Space=::NewEnumerator2
    Use Rotation Vector=false
   )
   &ShapeLocation(
    Cylinder Height=50
    Cylinder Height Midpoint=0.1
    Cylinder Radius=50
    MaskSpawnGroup=false
    Override Seed=false
    Rotation Mode=::NewEnumerator2
    Shape Primitive=::NewEnumerator1
    Surface Only=false
    UseOwnerScale=true
    Yaw / Pitch / Roll=<0 0.25 0>
   )
   &DLR_ProductPreviews(GameModeID:NiagaraInt32=$System.GameModeID, bPreviewVehicleSpeed:NiagaraBool=$System.bPreviewVehicleSpeed)
  ]
  particleUpdateScript [
   &ParticleState()
   &ScaleMeshSize(Scale Factor=&SelectVectorFromArray(
     Direct Array Index=$Particles.PartitionParticles.Partition
     Vector Selection Array=DataInterface() {propertyValues {FloatData [<1 -0.75 0.75>, <1 -0.75 -0.75>]}}
    )) {enabled=false}
   &InheritVelocity(Inherited Velocity Speed Limit=1, Speed Limit=true, UseSpeedThreshold=false)
   &InheritSourceMovement(
    Applied Position Scale=&Multiply_VectorByFloat(
     Vector=<1 1 1>
     Float=&FloatFromCurve(FloatCurve=DataInterface() {propertyValues {Curve {
         defaultValue=1
         keys [
          {arriveTangent=-1.7885, interpMode=RCIM_Cubic, leaveTangent=-1.7885, tangentMode=RCTM_User, time=0.05, value=1}
          {interpMode=RCIM_Cubic, tangentMode=RCTM_User, time=1, value=0}
         ]
        }}})
    )
    ApplyForce=false
    ApplyPosition=true
    ApplyPositionFalloff=false
   ) {enabled=false, module=InheritSourceMovement001}
   &ScaleColor(
    Scale RGB=&VectorFromFloat(Value=&MakeCustomFloatFromBool(
      Bool=$User.bIsPainted
      True Float=&FloatFromCurve(
       FloatCurve=DataInterface() {propertyValues {Curve {keys [{interpMode=RCIM_Cubic, leaveTangent=-1.6688, tangentMode=RCTM_Break, time=0, value=1.25}, <at 1 0.15 RCIM_Cubic>]}}}
       Scale Curve=4
      )
      False Float=&FloatFromCurve(
       FloatCurve=DataInterface() {propertyValues {Curve {keys [
           {arriveTangent=-1.9978, interpMode=RCIM_Cubic, leaveTangent=-1.9978, tangentMode=RCTM_User, time=0, value=1}
           {
            arriveTangentWeight=0.311
            interpMode=RCIM_Cubic
            leaveTangent=-2.5749
            leaveTangentWeight=0.0777
            tangentMode=RCTM_Break
            time=0.8
            value=0.4
           }
           {arriveTangentWeight=0.2828, interpMode=RCIM_Cubic, leaveTangentWeight=0.0471, tangentMode=RCTM_User, time=1, value=0.2}
          ]}}}
       Scale Curve=&RandomRangeFloat(Minimum=3, Maximum=4)
      )
     ))
    ScaleA=true
    ScaleRGB=true
   ) {enabled=false}
   &DynamicMaterialParameters(
    Index 0 Param 1=&MaskFloatBySpawnGroup(
     Masked Float=&FloatFromCurve(
      FloatCurve=DataInterface() {propertyValues {Curve {keys [
          <at 0 0 RCIM_Cubic>
          {arriveTangent=-0.1296, interpMode=RCIM_Cubic, leaveTangent=-0.1296, tangentMode=RCTM_User, time=0.65, value=0.42}
          {arriveTangent=0.644, tangentMode=RCTM_User, time=1, value=1}
         ]}}}
      Scale Curve=0.8
     )
     Passthrough Float=&FloatFromCurve(
      FloatCurve=DataInterface() {propertyValues {Curve {keys [
          <at 0.1 0 RCIM_Cubic>
          {arriveTangent=-0.1296, interpMode=RCIM_Cubic, leaveTangent=-0.1296, tangentMode=RCTM_User, time=0.65, value=0.42}
          {arriveTangent=0.644, tangentMode=RCTM_User, time=1, value=1}
         ]}}}
      Scale Curve=0.8
     )
    )
    Index 0 Param 2=&MaskFloatBySpawnGroup(
     Masked Float=&FloatFromCurve(FloatCurve=DataInterface() {propertyValues {Curve {keys [
          <at 0 1 RCIM_Cubic>
          {arriveTangent=-0.1355, interpMode=RCIM_Cubic, leaveTangent=-0.1355, tangentMode=RCTM_User, time=0.7, value=0.07}
          <at 1 0 RCIM_Cubic>
         ]}}})
     Passthrough Float=&FloatFromCurve(FloatCurve=DataInterface() {propertyValues {Curve {keys [
          <at 0 1.1 RCIM_Cubic>
          {arriveTangent=-0.1355, interpMode=RCIM_Cubic, leaveTangent=-0.1355, tangentMode=RCTM_User, time=0.7, value=0.07}
          <at 1 0 RCIM_Cubic>
         ]}}})
    )
    Index 1 Param 1=&MakeCustomFloatFromBool(Bool=$User.bIsPainted, True Float=0, False Float=1)
    Index 1 Param 2=&MakeCustomFloatFromBool(Bool=$User.bIsPainted, True Float=&MakeFloatFromInt(Int=$User.Lines_Glow), False Float=1)
    Module.Index0Param0Write=true
    Module.Index0Param1Write=true
    Module.Index0Param2Write=true
    Module.Index0Param3Write=true
    Module.Index1Param0Write=true
    Module.Index1Param1Write=true
    Module.Index1Param2Write=true
    Module.Index1Param3Write=true
    Write Parameter Index 1=true
   ) {module=DynamicMaterialParameters001}
   &SolveForcesAndVelocity(Clamp Velocity=false, Limit Acceleration=false)
   &ScaleMeshSizeBySpeed(Max Scale Factor=<12 6 6>, Speed Threshold=100)
  ]
  rendererClasses [&NiagaraMeshRendererProperties, &NiagaraSpriteRendererProperties]
  renderers [
   {rendererClass=&NiagaraMeshRendererProperties, rendererIndex=0}
   {rendererClass=&NiagaraMeshRendererProperties, rendererIndex=1}
   {rendererClass=&NiagaraMeshRendererProperties, rendererIndex=2}
   {rendererClass=&NiagaraMeshRendererProperties, rendererIndex=3}
   {rendererClass=&NiagaraSpriteRendererProperties, rendererIndex=4}
  ]
 }
 NE_VC_Boost_Curve3001 {
  emitterUpdateScript [
   &EmitterState()
   &VC_FilteredSocketCount()
   &SpawnRate(
    SpawnGroup=1
    SpawnRate=&Multiply_Float(
     A=$Engine.Emitter.SpawnCountScale
     B=&VC_Mask_Floats(Coro Value:NiagaraFloat=25, FrontEnd Value:NiagaraFloat=1, FN Value:NiagaraFloat=0.5)
    )
    Use Spawn Probability=false
   ) {module=SpawnRate003}
   &SpawnRate(SpawnRate=25, Use Spawn Probability=false)
   &SpawnRate(SpawnGroup=1, SpawnRate=50, Use Spawn Probability=false) {module=SpawnRate002}
  ]
  particleSpawnScript [
   &VC_PaintedVariantsOverride(DefaultColor=&SelectLinearColorFromArray(
     Array Sampling Mode=::NewEnumerator0
     Direct Array Index=?
     Color Selection Array=DataInterface() {propertyValues {ColorData [<rgba 0 1.6667 4 1>, <rgba 1.5 0 0.2 1>]}}
     Random Seed=0
    ))
   &InitializeParticle(
    Color=$Particles.Color
    Lifetime=&MaskFloatBySpawnGroup(
     Masked Float=&RandomRangeFloat(Minimum=0.1, Maximum=0.4)
     Passthrough Float=&RandomRangeFloat(Minimum=0.1, Maximum=0.3)
    )
    Mesh Scale=&Multiply_Vector(
     A=&MaskVectorBySpawnGroup(
      Masked Vector=&RandomRangeVector(Minimum=<7 6 6>, Maximum=<15 10 10>)
      Passthrough Vector=&RandomRangeVector(Minimum=<7 7 7>, Maximum=<9 9 9>)
     )
     B=&MakeVector(X=&RandomRangeFloat(Minimum=2, Maximum=3), Y=1, Z=1)
    )
    Mesh Scale Mode=::NewEnumerator4
    Sprite Rotation Mode=::NewEnumerator2
    UsePositionOffset=false
    Write Material Random=true
   )
   &PartitionParticles(Number Of Partitions=$Emitter.PortCount)
   &SkeletalMeshLocation(
    Bone / Socket Index=$Particles.PartitionParticles.Partition
    Bone / Socket Sampling Mode=::NewEnumerator5
    MaskSpawnGroup=false
    Mesh Sampling Type=::NewEnumerator6
    Sampled Position Offset=&MakeCustomVectorFromBool(Bool=$User.bIsFrontendPreview, True Vector=<32 0 0>, False Vector=<65 0 0>)
    Skeletal Mesh=$System.SkeletalMesh
   )
   &AddVelocity(Velocity=<0 0 0>) {enabled=false}
   /SetVariables(Particles.VisibilityTag=&SpawnGroup()) {isSetParameters=true, module=SetVariables_#6}
   &InitialMeshOrientation(
    Mesh Axis To Orient=&RandomRangeVector(Minimum=<0 -1 1>, Maximum=<0 1 -1>)
    Mesh Orientation Mode=::NewEnumerator2
    Orientation Vector=&RandomRangeVector(Minimum=<0 1 1>, Maximum=<0 -1 1>, Randomness Mode=::NewEnumerator1)
    Rotation=<90 65 0>
    Rotation Coordinate Space=::NewEnumerator2
    Use Rotation Vector=false
   )
   &ShapeLocation(
    Cylinder Height=50
    Cylinder Height Midpoint=0.1
    Cylinder Radius=50
    MaskSpawnGroup=false
    Override Seed=false
    Rotation Mode=::NewEnumerator2
    Shape Primitive=::NewEnumerator1
    Surface Only=false
    UseOwnerScale=true
    Yaw / Pitch / Roll=<0 0.25 0>
   )
   &DLR_ProductPreviews(GameModeID:NiagaraInt32=$System.GameModeID, bPreviewVehicleSpeed:NiagaraBool=$System.bPreviewVehicleSpeed)
  ]
  particleUpdateScript [
   &ParticleState()
   &ScaleMeshSize(Scale Factor=&SelectVectorFromArray(
     Direct Array Index=$Particles.PartitionParticles.Partition
     Vector Selection Array=DataInterface() {propertyValues {FloatData [<1 -0.75 0.75>, <1 -0.75 -0.75>]}}
    )) {enabled=false}
   &InheritVelocity(Inherited Velocity Speed Limit=1, Speed Limit=true, UseSpeedThreshold=false)
   &InheritSourceMovement(
    Applied Position Scale=&Multiply_VectorByFloat(
     Vector=<1 1 1>
     Float=&FloatFromCurve(FloatCurve=DataInterface() {propertyValues {Curve {
         defaultValue=1
         keys [
          {arriveTangent=-1.7885, interpMode=RCIM_Cubic, leaveTangent=-1.7885, tangentMode=RCTM_User, time=0.05, value=1}
          {interpMode=RCIM_Cubic, tangentMode=RCTM_User, time=1, value=0}
         ]
        }}})
    )
    ApplyForce=false
    ApplyPosition=true
    ApplyPositionFalloff=false
   ) {enabled=false, module=InheritSourceMovement001}
   &ScaleColor(
    Scale RGB=&VectorFromFloat(Value=&MakeCustomFloatFromBool(
      Bool=$User.bIsPainted
      True Float=&FloatFromCurve(
       FloatCurve=DataInterface() {propertyValues {Curve {keys [{interpMode=RCIM_Cubic, leaveTangent=-1.6688, tangentMode=RCTM_Break, time=0, value=1.25}, <at 1 0.15 RCIM_Cubic>]}}}
       Scale Curve=4
      )
      False Float=&FloatFromCurve(
       FloatCurve=DataInterface() {propertyValues {Curve {keys [
           {arriveTangent=-1.9978, interpMode=RCIM_Cubic, leaveTangent=-1.9978, tangentMode=RCTM_User, time=0, value=1}
           {
            arriveTangentWeight=0.311
            interpMode=RCIM_Cubic
            leaveTangent=-2.5749
            leaveTangentWeight=0.0777
            tangentMode=RCTM_Break
            time=0.8
            value=0.4
           }
           {arriveTangentWeight=0.2828, interpMode=RCIM_Cubic, leaveTangentWeight=0.0471, tangentMode=RCTM_User, time=1, value=0.2}
          ]}}}
       Scale Curve=&RandomRangeFloat(Minimum=3, Maximum=4)
      )
     ))
    ScaleA=true
    ScaleRGB=true
   ) {enabled=false}
   &DynamicMaterialParameters(
    Index 0 Param 1=&MaskFloatBySpawnGroup(
     Masked Float=&FloatFromCurve(
      FloatCurve=DataInterface() {propertyValues {Curve {keys [
          <at 0 0 RCIM_Cubic>
          {arriveTangent=-0.1296, interpMode=RCIM_Cubic, leaveTangent=-0.1296, tangentMode=RCTM_User, time=0.65, value=0.42}
          {arriveTangent=0.644, tangentMode=RCTM_User, time=1, value=1}
         ]}}}
      Scale Curve=0.8
     )
     Passthrough Float=&FloatFromCurve(
      FloatCurve=DataInterface() {propertyValues {Curve {keys [
          <at 0.1 0 RCIM_Cubic>
          {arriveTangent=-0.1296, interpMode=RCIM_Cubic, leaveTangent=-0.1296, tangentMode=RCTM_User, time=0.65, value=0.42}
          {arriveTangent=0.644, tangentMode=RCTM_User, time=1, value=1}
         ]}}}
      Scale Curve=0.8
     )
    )
    Index 0 Param 2=&MaskFloatBySpawnGroup(
     Masked Float=&FloatFromCurve(FloatCurve=DataInterface() {propertyValues {Curve {keys [
          <at 0 1 RCIM_Cubic>
          {arriveTangent=-0.1355, interpMode=RCIM_Cubic, leaveTangent=-0.1355, tangentMode=RCTM_User, time=0.7, value=0.07}
          <at 1 0 RCIM_Cubic>
         ]}}})
     Passthrough Float=&FloatFromCurve(FloatCurve=DataInterface() {propertyValues {Curve {keys [
          <at 0 1.1 RCIM_Cubic>
          {arriveTangent=-0.1355, interpMode=RCIM_Cubic, leaveTangent=-0.1355, tangentMode=RCTM_User, time=0.7, value=0.07}
          <at 1 0 RCIM_Cubic>
         ]}}})
    )
    Index 1 Param 1=&MakeCustomFloatFromBool(Bool=$User.bIsPainted, True Float=0, False Float=1)
    Index 1 Param 2=&MakeCustomFloatFromBool(Bool=$User.bIsPainted, True Float=&MakeFloatFromInt(Int=$User.Lines_Glow), False Float=1)
    Module.Index0Param0Write=true
    Module.Index0Param1Write=true
    Module.Index0Param2Write=true
    Module.Index0Param3Write=true
    Module.Index1Param0Write=true
    Module.Index1Param1Write=true
    Module.Index1Param2Write=true
    Module.Index1Param3Write=true
    Write Parameter Index 1=true
   ) {module=DynamicMaterialParameters001}
   &SolveForcesAndVelocity(Clamp Velocity=false, Limit Acceleration=false)
   &ScaleMeshSizeBySpeed(Max Scale Factor=<12 6 6>, Speed Threshold=100)
  ]
  rendererClasses [&NiagaraMeshRendererProperties, &NiagaraSpriteRendererProperties]
  renderers [
   {rendererClass=&NiagaraMeshRendererProperties, rendererIndex=0}
   {rendererClass=&NiagaraMeshRendererProperties, rendererIndex=1}
   {rendererClass=&NiagaraMeshRendererProperties, rendererIndex=2}
   {rendererClass=&NiagaraMeshRendererProperties, rendererIndex=3}
   {rendererClass=&NiagaraSpriteRendererProperties, rendererIndex=4}
  ]
 }
 NE_VC_Boost_Ribbon_Parent {
  emitterUpdateScript [
   &EmitterState()
   &VC_FilteredSocketCount()
   &VC_RateQualityHandler(Default Rate:NiagaraFloat=&Multiply_Float(
     A=100
     B=&VC_Mask_Floats(Coro Value:NiagaraFloat=1, FrontEnd Value:NiagaraFloat=0.75, FN Value:NiagaraFloat=0.75)
    ))
   &VC_SpawnRibbon_v1(Number of Ribbons=$Emitter.PortCount, SpawnRate=$Emitter.Spawn Rate)
  ]
  particleSpawnScript [
   &VC_PaintedVariantsOverride(
    Black=<rgba 0 0 0 1>
    BurntSienna=<rgba 1 0.3394 0 1>
    Cobalt=<rgba 0.0312 0.0492 1 1>
    Crimson=<rgba 0.2261 0 0.0061 1>
    ForestGreen=<rgba 0.0114 0.6146 0.0065 1>
    Grey=<rgba 0.026 0.026 0.026 1>
    Orange=<rgba 1 0.1779 0.0103 1>
    Pink=<rgba 0.9844 0.1179 0.6693 1>
    Purple=<rgba 0.6308 0.0006 0.9911 1>
   )
   &VC_InitializeRibbon_v1(
    Color=$Particles.Color
    Enable Skeletal Position=true
    Lifetime=0.19
    Offset Amount:NiagaraFloat=&MakeCustomFloatFromBool(Bool=$User.bIsFrontendPreview, True Float=0, False Float=50)
    Offset Axis:ENiagaraOrientationAxis=::XAxis
    Ribbon Width=35
    Write Material Random=false
   )
   &AddVelocity(Velocity=&VC_Mask_Vectors(Coro Value:Vector3f=<-2400 0 0>, FrontEnd Value:Vector3f=<-1200 0 0>, FN Value:Vector3f=<-400 0 0>)) {module=AddVelocity001}
   &DLR_ProductPreviews(GameModeID:NiagaraInt32=$System.GameModeID, bPreviewVehicleSpeed:NiagaraBool=$System.bPreviewVehicleSpeed)
  ]
  particleUpdateScript [
   &ParticleState()
   &InheritSourceMovement(
    Applied Position Scale=&VectorFromFloat(Value=&FloatFromCurve(FloatCurve=DataInterface() {propertyValues {Curve {
         defaultValue=1
         keys [
          {arriveTangent=-1.7885, interpMode=RCIM_Cubic, leaveTangent=-1.7885, tangentMode=RCTM_User, time=0.05, value=1}
          {interpMode=RCIM_Cubic, tangentMode=RCTM_User, time=1, value=0}
         ]
        }}}))
    ApplyForce=false
    ApplyPosition=true
    ApplyPositionFalloff=false
   )
   &SolveForcesAndVelocity(Clamp Velocity=false, Limit Acceleration=false)
   &VC_UpdateRibbon_v1(Enable Skeletal Position=true)
   &ScaleColor(
    Color Scale:LinearColor=&MakeCustomLinearColorFromBool(
     Bool=$User.bIsPainted
     True Color=<rgba 1 1 1 1>
     False Color=&ColorFromCurve(
      ColorCurve=DataInterface() {propertyValues {
        AlphaCurve {keys [<at 0 1 RCIM_Constant>]}
        BlueCurve {keys [<at 0.3426 1 RCIM_Constant>,<at 0.4755 0.7656 RCIM_Constant>]}
        GreenCurve {keys [<at 0.3426 0.0999 RCIM_Constant>,<at 0.4755 0.0383 RCIM_Constant>]}
        RedCurve {keys [<at 0.3426 1 RCIM_Constant>,<at 0.4755 0 RCIM_Constant>]}
       }}
      CurveIndex=&Sine(Normalized Angle=$Emitter.Age, Period=0.2, Scale=1, Bias=0)
     )
    )
    Scale Mode=3
   )
   &ScaleRibbonWidth(Ribbon Width Scale=&FloatFromCurve(FloatCurve=DataInterface() {propertyValues {Curve {keys [
         {arriveTangent=-0.0268, interpMode=RCIM_Cubic, leaveTangent=-0.0268, tangentMode=RCTM_User, time=0, value=0}
         <at 0.1014 1.0118 RCIM_Cubic>
         {arriveTangent=0.1194, tangentMode=RCTM_User, time=1, value=0}
        ]}}}))
  ]
  rendererClasses [&NiagaraRibbonRendererProperties]
  renderers [{rendererClass=&NiagaraRibbonRendererProperties}]
 }
 NE_VC_Boost_ThickC3 {
  simTarget=GPUComputeSim
  emitterUpdateScript [
   &EmitterState()
   &SpawnRate(SpawnGroup=1, SpawnRate=100, Use Spawn Probability=false) {module=SpawnRate002}
   &SpawnRate(
    SpawnGroup=1
    SpawnRate=&Multiply_Float(
     A=$Engine.Emitter.SpawnCountScale
     B=&VC_Mask_Floats(Coro Value:NiagaraFloat=75, FrontEnd Value:NiagaraFloat=1, FN Value:NiagaraFloat=0.5)
    )
    Use Spawn Probability=false
   ) {module=SpawnRate003}
   &VC_FilteredSocketCount()
   &SpawnRate(SpawnRate=50, Use Spawn Probability=false)
   /SetVariables(Emitter.Density:NiagaraFloat=&MakeCustomFloatFromBool(Bool=$User.bIsPainted, True Float=0.1, False Float=2.3)) {isSetParameters=true}
  ]
  particleSpawnScript [
   &VC_PaintedVariantsOverride(
    Black=<rgba 0 0 0 1>
    BurntSienna=&RandomRangeLinearColor(Minimum=<rgba 1 0.3394 0 1>, Maximum=<rgba 0.063 0.0091 0 1>)
    Cobalt=&RandomRangeLinearColor(Minimum=<rgba 0.001 0.2291 0.7969 1>, Maximum=<rgba 0 0.0028 0.151 1>)
    Crimson=&RandomRangeLinearColor(Minimum=<rgba 0.0885 0.0004 0 1>, Maximum=<rgba 0.6979 0.0028 0.0026 1>)
    DefaultColor=&SelectLinearColorFromArray(
     Array Sampling Mode=::NewEnumerator0
     Direct Array Index=?
     Color Selection Array=DataInterface() {propertyValues {ColorData [<rgba 0 1.6667 4 1>, <rgba 1.5 0 0.2 1>]}}
     Random Seed=0
    )
    ForestGreen=&RandomRangeLinearColor(Minimum=<rgba 0.001 0.0521 0.0006 1>, Maximum=<rgba 0 1.2 0.0606 1>)
    Grey=&RandomRangeLinearColor(Minimum=<rgba 0.007 0.007 0.0061 1>, Maximum=<rgba 0.07 0.07 0.07 1>)
    Lime=&RandomRangeLinearColor(Minimum=<rgba 0.0232 0.1043 0 1>, Maximum=<rgba 0.4089 0.8177 0 1>)
    Orange=&RandomRangeLinearColor(Minimum=<rgba 0.4397 0.0044 0 1>, Maximum=<rgba 1 0.1779 0.0103 1>)
    Pink=&RandomRangeLinearColor(Minimum=<rgba 0.1845 0.0037 0.0222 1>, Maximum=<rgba 0.9823 0.1779 0.9911 1>)
    Purple=&RandomRangeLinearColor(Minimum=<rgba 0.0595 0.0003 0.6654 1>, Maximum=<rgba 0.5391 0.0003 0.9911 1>)
    Saffron=&RandomRangeLinearColor(Minimum=<rgba 0.2135 0.2135 0.0014 1>, Maximum=<rgba 1 1 0.0048 1>)
    SkyBlue=&RandomRangeLinearColor(Minimum=<rgba 0.0021 0.0482 0.0742 1>, Maximum=<rgba 0.0497 1 0.9911 1>)
    TitaniumWhite=&RandomRangeLinearColor(Minimum=<rgba 0.3021 0.3021 0.3021 1>, Maximum=<rgba 0.9304 0.9304 0.9304 1>)
   )
   &InitializeParticle(
    Color Mode=::NewEnumerator0
    Lifetime Max=0.3
    Lifetime Min=0.05
    Lifetime Mode=::NewEnumerator1
    Mesh Scale=&Multiply_Vector(
     A=&MaskVectorBySpawnGroup(
      Masked Vector=&RandomRangeVector(Minimum=<3 4 4>, Maximum=<6 4 4>)
      Passthrough Vector=&RandomRangeVector(Minimum=<7 5 5>, Maximum=<12 10 10>)
     )
     B=&MakeVector(X=2, Y=1, Z=1)
    )
    Mesh Scale Mode=::NewEnumerator4
    UsePositionOffset=false
    Write Material Random=true
   )
   &PartitionParticles(Number Of Partitions=$Emitter.PortCount)
   &SkeletalMeshLocation(
    Bone / Socket Index=$Particles.PartitionParticles.Partition
    Bone / Socket Sampling Mode=::NewEnumerator5
    MaskSpawnGroup=false
    Mesh Sampling Type=::NewEnumerator6
    Sampled Position Offset=&MakeCustomVectorFromBool(Bool=$User.bIsFrontendPreview, True Vector=<25 0 0>, False Vector=<65 0 0>)
    Skeletal Mesh=$System.SkeletalMesh
   )
   &AddVelocity(
    Cone Angle=180
    Cone Axis=<-1 0 0>
    Module.UseConeFalloff=false
    Module.UseVelDistribution=false
    Override Seed=false
    Velocity Mode=::NewEnumerator2
    Velocity Speed=100
   ) {enabled=false}
   &InitialMeshOrientation(
    Mesh Orientation Mode=::NewEnumerator2
    Rotation=&MaskVectorBySpawnGroup(
     Masked Vector=&RandomRangeVector(Minimum=<-0.1 -0.1 -0.15>, Maximum=<0.1 0.1 0.15>)
     Passthrough Vector=&RandomRangeVector(Minimum=<-0.02 -0.02 -0.05>, Maximum=<0.02 0.042 0.05>)
    )
    Rotation Coordinate Space=::NewEnumerator2
    Use Rotation Vector=true
   )
   &ShapeLocation(
    Cylinder Height=50
    Cylinder Height Midpoint=0.1
    Cylinder Radius=25
    MaskSpawnGroup=false
    Override Seed=false
    Rotation Mode=::NewEnumerator2
    Shape Primitive=::NewEnumerator1
    Surface Only=false
    UseOwnerScale=true
    Yaw / Pitch / Roll=<0 0.25 0>
   )
   &DLR_ProductPreviews(GameModeID:NiagaraInt32=$System.GameModeID, bPreviewVehicleSpeed:NiagaraBool=$System.bPreviewVehicleSpeed)
  ]
  particleUpdateScript [
   &ParticleState()
   &ScaleMeshSize(Scale Factor=&SelectVectorFromArray(
     Direct Array Index=$Particles.PartitionParticles.Partition
     Vector Selection Array=DataInterface() {propertyValues {FloatData [<1 -0.75 0.75>, <1 -0.75 -0.75>]}}
    )) {enabled=false}
   &InheritSourceMovement(
    Applied Position Scale=&Multiply_VectorByFloat(
     Vector=<1 1 1>
     Float=&FloatFromCurve(FloatCurve=DataInterface() {propertyValues {Curve {
         defaultValue=1
         keys [
          {arriveTangent=-1.7885, interpMode=RCIM_Cubic, leaveTangent=-1.7885, tangentMode=RCTM_User, time=0.05, value=1}
          {interpMode=RCIM_Cubic, tangentMode=RCTM_User, time=1, value=0}
         ]
        }}})
    )
    ApplyForce=false
    ApplyPosition=true
    ApplyPositionFalloff=false
   ) {enabled=false, module=InheritSourceMovement001}
   &ScaleColor(
    Scale RGB=&VectorFromFloat(Value=&FloatFromCurve(FloatCurve=DataInterface() {propertyValues {Curve {keys [<at 0 10>,<at 0.75 1>]}}}))
    ScaleA=true
    ScaleRGB=true
   ) {enabled=false}
   &DynamicMaterialParameters(
    Index 0 Param 1=&FloatFromCurve(
     FloatCurve=DataInterface() {propertyValues {Curve {keys [
         <at 0.1 0 RCIM_Cubic>
         {arriveTangent=-0.1296, interpMode=RCIM_Cubic, leaveTangent=-0.1296, tangentMode=RCTM_User, time=0.65, value=0.42}
         {arriveTangent=0.644, tangentMode=RCTM_User, time=1, value=1}
        ]}}}
     Scale Curve=0.8
    )
    Index 0 Param 2=&FloatFromCurve(FloatCurve=DataInterface() {propertyValues {Curve {keys [
         <at 0 1.1 RCIM_Cubic>
         {arriveTangent=-0.1355, interpMode=RCIM_Cubic, leaveTangent=-0.1355, tangentMode=RCTM_User, time=0.7, value=0.07}
         <at 1 0 RCIM_Cubic>
        ]}}})
    Index 1 Param 1=&MakeCustomFloatFromBool(Bool=$User.bIsPainted, True Float=0, False Float=1)
    Index 1 Param 2=&MakeCustomFloatFromBool(Bool=$User.bIsPainted, True Float=&MakeFloatFromInt(Int=$User.Lines_Glow), False Float=1)
    Module.Index0Param0Write=true
    Module.Index0Param1Write=true
    Module.Index0Param2Write=true
    Module.Index0Param3Write=true
    Module.Index1Param0Write=true
    Module.Index1Param1Write=true
    Module.Index1Param2Write=true
    Module.Index1Param3Write=true
    Write Parameter Index 1=true
   ) {module=DynamicMaterialParameters001}
   &SolveForcesAndVelocity(Clamp Velocity=false, Limit Acceleration=false)
   &ScaleMeshSizeBySpeed(Max Scale Factor=<4 5 5>, Speed Threshold=100)
  ]
  rendererClasses [&NiagaraMeshRendererProperties, &NiagaraSpriteRendererProperties]
  renderers [
   {rendererClass=&NiagaraMeshRendererProperties, rendererIndex=0}
   {rendererClass=&NiagaraMeshRendererProperties, rendererIndex=1}
   {rendererClass=&NiagaraSpriteRendererProperties, rendererIndex=2}
  ]
 }
 NE_VC_Boost_ThickC3001 {
  emitterUpdateScript [
   &EmitterState()
   &SpawnRate(SpawnGroup=1, SpawnRate=50, Use Spawn Probability=false) {module=SpawnRate002}
   &SpawnRate(
    SpawnGroup=1
    SpawnRate=&Multiply_Float(
     A=$Engine.Emitter.SpawnCountScale
     B=&VC_Mask_Floats(Coro Value:NiagaraFloat=25, FrontEnd Value:NiagaraFloat=1, FN Value:NiagaraFloat=0.5)
    )
    Use Spawn Probability=false
   ) {module=SpawnRate003}
   &VC_FilteredSocketCount()
   &SpawnRate(SpawnRate=25, Use Spawn Probability=false)
  ]
  particleSpawnScript [
   &VC_PaintedVariantsOverride(DefaultColor=&SelectLinearColorFromArray(
     Array Sampling Mode=::NewEnumerator0
     Direct Array Index=?
     Color Selection Array=DataInterface() {propertyValues {ColorData [<rgba 0 1.6667 4 1>, <rgba 1.5 0 0.2 1>]}}
     Random Seed=0
    ))
   &InitializeParticle(
    Color=$Particles.Color
    Lifetime Max=0.3
    Lifetime Min=0.05
    Lifetime Mode=::NewEnumerator1
    Mesh Scale=&Multiply_Vector(
     A=&MaskVectorBySpawnGroup(
      Masked Vector=&RandomRangeVector(Minimum=<3 4 4>, Maximum=<6 4 4>)
      Passthrough Vector=&RandomRangeVector(Minimum=<7 5 5>, Maximum=<12 10 10>)
     )
     B=&MakeVector(X=2, Y=1, Z=1)
    )
    Mesh Scale Mode=::NewEnumerator4
    UsePositionOffset=false
    Write Material Random=true
   )
   &PartitionParticles(Number Of Partitions=$Emitter.PortCount)
   &SkeletalMeshLocation(
    Bone / Socket Index=$Particles.PartitionParticles.Partition
    Bone / Socket Sampling Mode=::NewEnumerator5
    MaskSpawnGroup=false
    Mesh Sampling Type=::NewEnumerator6
    Sampled Position Offset=&MakeCustomVectorFromBool(Bool=$User.bIsFrontendPreview, True Vector=<25 0 0>, False Vector=<65 0 0>)
    Skeletal Mesh=$System.SkeletalMesh
   )
   &AddVelocity(
    Cone Angle=180
    Cone Axis=<-1 0 0>
    Module.UseConeFalloff=false
    Module.UseVelDistribution=false
    Override Seed=false
    Velocity Mode=::NewEnumerator2
    Velocity Speed=100
   ) {enabled=false}
   &InitialMeshOrientation(
    Mesh Orientation Mode=::NewEnumerator2
    Rotation=&MaskVectorBySpawnGroup(
     Masked Vector=&RandomRangeVector(Minimum=<-0.1 -0.1 -0.15>, Maximum=<0.1 0.1 0.15>)
     Passthrough Vector=&RandomRangeVector(Minimum=<-0.02 -0.02 -0.05>, Maximum=<0.02 0.042 0.05>)
    )
    Rotation Coordinate Space=::NewEnumerator2
    Use Rotation Vector=true
   )
   &ShapeLocation(
    Cylinder Height=50
    Cylinder Height Midpoint=0.1
    Cylinder Radius=25
    MaskSpawnGroup=false
    Override Seed=false
    Rotation Mode=::NewEnumerator2
    Shape Primitive=::NewEnumerator1
    Surface Only=false
    UseOwnerScale=true
    Yaw / Pitch / Roll=<0 0.25 0>
   )
   &DLR_ProductPreviews(GameModeID:NiagaraInt32=$System.GameModeID, bPreviewVehicleSpeed:NiagaraBool=$System.bPreviewVehicleSpeed)
  ]
  particleUpdateScript [
   &ParticleState()
   &ScaleMeshSize(Scale Factor=&SelectVectorFromArray(
     Direct Array Index=$Particles.PartitionParticles.Partition
     Vector Selection Array=DataInterface() {propertyValues {FloatData [<1 -0.75 0.75>, <1 -0.75 -0.75>]}}
    )) {enabled=false}
   &InheritSourceMovement(
    Applied Position Scale=&Multiply_VectorByFloat(
     Vector=<1 1 1>
     Float=&FloatFromCurve(FloatCurve=DataInterface() {propertyValues {Curve {
         defaultValue=1
         keys [
          {arriveTangent=-1.7885, interpMode=RCIM_Cubic, leaveTangent=-1.7885, tangentMode=RCTM_User, time=0.05, value=1}
          {interpMode=RCIM_Cubic, tangentMode=RCTM_User, time=1, value=0}
         ]
        }}})
    )
    ApplyForce=false
    ApplyPosition=true
    ApplyPositionFalloff=false
   ) {enabled=false, module=InheritSourceMovement001}
   &ScaleColor(
    Scale RGB=&VectorFromFloat(Value=&FloatFromCurve(FloatCurve=DataInterface() {propertyValues {Curve {keys [<at 0 10>,<at 0.75 1>]}}}))
    ScaleA=true
    ScaleRGB=true
   ) {enabled=false}
   &DynamicMaterialParameters(
    Index 0 Param 1=&FloatFromCurve(
     FloatCurve=DataInterface() {propertyValues {Curve {keys [
         <at 0.1 0 RCIM_Cubic>
         {arriveTangent=-0.1296, interpMode=RCIM_Cubic, leaveTangent=-0.1296, tangentMode=RCTM_User, time=0.65, value=0.42}
         {arriveTangent=0.644, tangentMode=RCTM_User, time=1, value=1}
        ]}}}
     Scale Curve=0.8
    )
    Index 0 Param 2=&FloatFromCurve(FloatCurve=DataInterface() {propertyValues {Curve {keys [
         <at 0 1.1 RCIM_Cubic>
         {arriveTangent=-0.1355, interpMode=RCIM_Cubic, leaveTangent=-0.1355, tangentMode=RCTM_User, time=0.7, value=0.07}
         <at 1 0 RCIM_Cubic>
        ]}}})
    Index 1 Param 1=&MakeCustomFloatFromBool(Bool=$User.bIsPainted, True Float=0, False Float=1)
    Index 1 Param 2=&MakeCustomFloatFromBool(Bool=$User.bIsPainted, True Float=&MakeFloatFromInt(Int=$User.Lines_Glow), False Float=1)
    Module.Index0Param0Write=true
    Module.Index0Param1Write=true
    Module.Index0Param2Write=true
    Module.Index0Param3Write=true
    Module.Index1Param0Write=true
    Module.Index1Param1Write=true
    Module.Index1Param2Write=true
    Module.Index1Param3Write=true
    Write Parameter Index 1=true
   ) {module=DynamicMaterialParameters001}
   &SolveForcesAndVelocity(Clamp Velocity=false, Limit Acceleration=false)
   &ScaleMeshSizeBySpeed(Max Scale Factor=<4 5 5>, Speed Threshold=100)
  ]
  rendererClasses [&NiagaraMeshRendererProperties, &NiagaraSpriteRendererProperties]
  renderers [
   {rendererClass=&NiagaraMeshRendererProperties, rendererIndex=0}
   {rendererClass=&NiagaraMeshRendererProperties, rendererIndex=1}
   {rendererClass=&NiagaraSpriteRendererProperties, rendererIndex=2}
  ]
 }
}
summary {
 systemName=NS_Boost_Animatedlines
 emitters [
  {
   emitterName=NE_VC_Boost_ThickC3
   simTarget=GPUComputeSim
   rendererClasses [&NiagaraMeshRendererProperties, &NiagaraSpriteRendererProperties]
  }
  {
   emitterName=NE_VC_Boost_Curve3
   simTarget=GPUComputeSim
   rendererClasses [&NiagaraMeshRendererProperties, &NiagaraSpriteRendererProperties]
  }
  {
   emitterName=NE_VC_Boost_Curve2
   simTarget=GPUComputeSim
   rendererClasses [&NiagaraMeshRendererProperties, &NiagaraSpriteRendererProperties]
  }
  {
   emitterName=NE_VC_Boost_C3Thin
   rendererClasses [&NiagaraMeshRendererProperties, &NiagaraSpriteRendererProperties]
  }
  {emitterName=NE_VC_Boost_Ribbon_Parent, rendererClasses [&NiagaraRibbonRendererProperties]}
  {
   emitterName=NE_VC_Boost_ThickC3001
   rendererClasses [&NiagaraMeshRendererProperties, &NiagaraSpriteRendererProperties]
  }
  {
   emitterName=NE_VC_Boost_Curve3001
   rendererClasses [&NiagaraMeshRendererProperties, &NiagaraSpriteRendererProperties]
  }
  {
   emitterName=NE_VC_Boost_Curve2001
   rendererClasses [&NiagaraMeshRendererProperties, &NiagaraSpriteRendererProperties]
  }
  {
   emitterName=NE_VC_Boost_C3Thin001
   rendererClasses [&NiagaraMeshRendererProperties, &NiagaraSpriteRendererProperties]
  }
 ]
 userVariables [
  User.bIsFrontendPreview {type=NiagaraBool}
  User.bIsPainted {type=NiagaraBool, defaultValue=true}
  User.blsIsCoroColor {type=NiagaraBool}
  User.bPreviewVehicleSpeed {type=NiagaraBool}
  User.ColorVariantID {type=NiagaraInt32, defaultValue=3}
  User.Density {type=NiagaraFloat, defaultValue=0, description=""}
  User.GameModeID {type=NiagaraInt32, defaultValue=3}
  User.Lines_Glow {type=NiagaraInt32, defaultValue=1}
  User.PaintedColor {type=LinearColor, description="", defaultValue {struct=LinearColor, value=<rgba 1 1 1 1>}}
  User.PaintedScalar {type=NiagaraFloat, defaultValue=1, description=""}
  User.PortCount {type=NiagaraInt32, defaultValue=2}
 ]
}
